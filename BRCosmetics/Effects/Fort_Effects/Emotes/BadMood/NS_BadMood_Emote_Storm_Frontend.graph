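# Niagara storm for the BadMood emote, apparently its frontend variant: a rotating cloud mesh that grows and fades in, with rain, lightning and splashes.
# Looping lightning meshes with one-shot audio share a system-variable loop delay with light flashes, alongside separately delayed one-time bursts of each.
# Zero-size proxy sprites fall alongside the collision-disabled rain and generate collision events, apparently to drive the splash emitter, which lacks its own spawn module.
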NiagaraSystem /BRCosmetics/Effects/Fort_Effects/Emotes/BadMood/NS_BadMood_Emote_Storm_Frontend
dependencies {
 usedDataInterfaces [
  &NiagaraDataInterfaceVector2DCurve
  &NiagaraDataInterfaceCurve
  &NiagaraDataInterfaceAudioPlayer
  &NiagaraDataInterfaceVectorCurve
  &NiagaraDataInterfaceColorCurve
 ]
 usedDynamicInputs []
 usedModules [
  &self:SystemSpawnScript.SystemScriptSource.SystemScriptGraph.NiagaraNodeAssignment_1.SetVariables_#1
  &SystemState
 ]
}
emitters {
 CloudLoop_Game {
  emitterUpdateScript [
   &EmitterState(
    Inactive Response=::NewEnumerator1
    Life Cycle Mode=::NewEnumerator1
    Loop Behavior=::NewEnumerator1
    Loop Duration Mode=::NewEnumerator1
    UseLoopDelay=false
   )
   &SpawnBurst_Instantaneous(Spawn Count=1, Use Spawn Probability=true)
  ]
  particleSpawnScript [
   /Niagara/Modules/Spawn/Initialization/InitializeParticle(
    Color=<rgba 0.7083 0.9805 1 1>
    Mass=&UniformRangedFloat(Minimum=0.4, Maximum=2)
    Mesh Scale=&VectorFromFloat(Value=&MakeFloatFromVector(Vector=<1.3 0 0>, Channel=::NewEnumerator0))
    Module.Write Color=true
    Module.Write Lifetime=true
    Module.Write Mass=false
    Module.Write Material Random=true
    Module.Write Position=false
    Module.Write Scale=true
    Module.Write SpriteRotation=true
    Module.Write SpriteSize=false
    Position=&SimulationPosition()
    Sprite Rotation=&NormalizedAngleToDegrees(NormalizedAngle=&UniformRangedFloat(Minimum=0, Maximum=1))
    Sprite Size=<100 100>
   )
   &MeshRotationForce(Rotation=&UniformRangedVector(Minimum=<0 0 -0.03>, Maximum=<0 0 -0.03>))
   &DynamicMaterialParameters(
    Index 0 Param 2=1
    Index 0 Param 4=&FloatFromCurve(FloatCurve=DataInterface() {propertyValues {Curve {keys [{arriveTangent=2.1413, interpMode=RCIM_Cubic, leaveTangent=2.1413, tangentMode=RCTM_User, time=0, value=0}, <at 1 1 RCIM_Cubic>]}}})
    Module.Param0WriteEnabled=true
    Module.Param1WriteEnabled=true
    Module.Param2WriteEnabled=true
    Module.Param3WriteEnabled=true
   ) {module=DynamicMaterialParameters001}
   &ApplyInitialForces()
  ]
  particleUpdateScript [
   &ParticleState(Kill Particles When Lifetime Has Elapsed=false)
   &Color(Color=&ColorFromCurve(ColorCurve=DataInterface() {propertyValues {
       AlphaCurve {keys [
         {arriveTangent=0.356, interpMode=RCIM_Cubic, leaveTangent=0.356, tangentMode=RCTM_User, time=0, value=0}
         {arriveTangent=-0.0386, leaveTangent=-0.0386, time=0.5, value=0.45}
        ]}
       BlueCurve {keys [<at -0.0021 0>,<at 0.5 1>]}
       GreenCurve {keys [<at -0.0021 0>,<at 0.5 0.9805>]}
       RedCurve {keys [<at -0.0021 0>,<at 0.5 0.7083>]}
      }})) {module=Color001}
   &ScaleMeshSize(Scale Factor=&VectorFromFloat(Value=&MakeFloatFromVector(
      Vector=&VectorFromCurve(VectorCurve=DataInterface() {propertyValues {
         XCurve {keys [
           {arriveTangent=1.0549, interpMode=RCIM_Cubic, time=0, value=0.25}
           {arriveTangent=0.0199, interpMode=RCIM_Cubic, leaveTangent=0.0199, tangentMode=RCTM_User, time=0.2661, value=1}
          ]}
         YCurve {keys [<at -0.0011 1.0067>,<at 0.2039 1.0067>]}
         ZCurve {keys [<at -0.0011 1.0067>,<at 0.2039 1.0067>]}
        }})
      Channel=::NewEnumerator0
     )))
   &SolveForcesAndVelocity(Clamp Velocity=false, Limit Acceleration=false)
   &SolveRotationalForcesAndVelocity()
   &ScaleMeshSize(Scale Factor=&self:ScratchDynamicInput()) {module=ScaleMeshSize001}
   /Niagara/Modules/Update/Orientation/V2/UpdateMeshOrientation(Rotation Vector:Vector3f=&VectorFromCurve(VectorCurve=DataInterface() {propertyValues {
       XCurve {keys [{interpMode=RCIM_Cubic, tangentMode=RCTM_User, time=0, value=0}, {interpMode=RCIM_Cubic, tangentMode=RCTM_User, time=1, value=0}]}
       YCurve {keys [{interpMode=RCIM_Cubic, tangentMode=RCTM_User, time=0, value=0}, {interpMode=RCIM_Cubic, tangentMode=RCTM_User, time=1, value=0}]}
       ZCurve {keys [
         {arriveTangent=2.0937, interpMode=RCIM_Cubic, tangentMode=RCTM_User, time=0, value=-0.2}
         {interpMode=RCIM_Cubic, tangentMode=RCTM_User, time=1, value=-0.03}
        ]}
      }}))
  ]
  rendererClasses [&NiagaraMeshRendererProperties]
  renderers [{rendererClass=&NiagaraMeshRendererProperties, rendererIndex=0}]
 }
 Collision_Proxy {
  emitterUpdateScript [
   &EmitterState(
    Delay First Loop Only=true
    Inactive Response=::NewEnumerator1
    Life Cycle Mode=::NewEnumerator1
    Loop Delay=0.5
    Loop Duration=1
    UseLoopDelay=true
   )
   &SpawnRate(SpawnRate=30, Use Spawn Probability=false)
  ]
  particleSpawnScript [
   &InitializeParticle(
    Color=<rgba 1 1 1 0.5>
    Lifetime=&UniformRangedFloat(Minimum=0.6, Maximum=0.8)
    Mass=&UniformRangedFloat(Minimum=0.75, Maximum=1.75)
    Module.Write Color=true
    Module.Write Lifetime=true
    Module.Write Mass=false
    Module.Write Material Random=true
    Module.Write Position=false
    Module.Write Scale=false
    Module.Write SpriteRotation=false
    Module.Write SpriteSize=true
    Position=&SimulationPosition()
    Sprite Size=&Vector2DFromFloat(Value=&MakeFloatFromVector(Vector=&UniformRangedVector(Minimum=<0 0 0>, Maximum=<0 0 0>), Channel=::NewEnumerator0))
   )
   /Niagara/Modules/Spawn/Location/SphereLocation(Hemisphere Z=true, MaskSpawnGroup=false, Module.Override Randomness=false, Non Uniform Scale=<1 1 -1>, Sphere Radius=28, Surface Only=false)
  ]
  particleUpdateScript [
   &ParticleState()
   &GravityForce()
   &Collision(
    Advanced Aging Rate=1
    Age Colliding Particles=false
    Analytical Collision Plane Position 1:Vector3f=<0 0 -200>
    CPU Collision Type=::NewEnumerator3
    Module.Control Rest State Via Surface Normal=false
    Module.EnableMaxCollisionCount=false
    Module.Randomize Collision Normal Vector=false
    Perform 2 Analytical Collision Queries=false
    Radius Calculation Type=::NewEnumerator1
   )
   &SolveForcesAndVelocity(Clamp Velocity=false, Limit Acceleration=false)
   &ScaleSpriteSize(Scale Factor=&Vector2DFromCurve(
     Vector2Curve=DataInterface() {propertyValues {
       XCurve {keys [
         {arriveTangent=5.9728, interpMode=RCIM_Cubic, leaveTangent=5.9728, tangentMode=RCTM_User, time=0, value=0.5817}
         {arriveTangent=-0.1, interpMode=RCIM_Cubic, leaveTangent=-0.1, tangentMode=RCTM_User, time=0.173, value=0.9552}
         {interpMode=RCIM_Cubic, tangentMode=RCTM_User, time=1, value=0.1086}
        ]}
       YCurve {keys [
         {arriveTangent=6.2715, interpMode=RCIM_Cubic, leaveTangent=6.2715, tangentMode=RCTM_User, time=0, value=0.5817}
         {arriveTangent=1.8836, interpMode=RCIM_Cubic, leaveTangent=1.8836, tangentMode=RCTM_User, time=0.209, value=1.5556}
         {arriveTangent=0.0517, interpMode=RCIM_Cubic, leaveTangent=0.0517, tangentMode=RCTM_User, time=1, value=2.8163}
        ]}
      }}
     ScaleCurve=<1 1>
    ))
   /Niagara/Modules/Events/GenerateCollisionEvent(Module.Use Time Between Events=false, Module.Use Velocity Threshold=true)
  ]
  rendererClasses [&NiagaraSpriteRendererProperties]
  renderers [{rendererClass=&NiagaraSpriteRendererProperties}]
 }
 Collision_Splashes {
  emitterUpdateScript [&EmitterState(
    Delay First Loop Only=true
    Inactive Response=::NewEnumerator1
    Life Cycle Mode=::NewEnumerator1
    Loop Delay=0.5
    Loop Duration=1
    UseLoopDelay=true
   )]
  particleSpawnScript [
   &InitializeParticle(
    Color=<rgba 1 1 1 0.5>
    Lifetime=&UniformRangedFloat(Minimum=0.2, Maximum=0.4)
    Mass=&UniformRangedFloat(Minimum=0.5, Maximum=1)
    Module.Write Color=true
    Module.Write Lifetime=true
    Module.Write Mass=false
    Module.Write Material Random=true
    Module.Write Position=false
    Module.Write Scale=false
    Module.Write SpriteRotation=false
    Module.Write SpriteSize=true
    Position=&SimulationPosition()
    Sprite Size=&Vector2DFromFloat(Value=&MakeFloatFromVector(Vector=&UniformRangedVector(Minimum=<1.5 0 0>, Maximum=<3 0 0>), Channel=::NewEnumerator0))
   )
   &SphereLocation(Hemisphere Z=true, MaskSpawnGroup=false, Module.Override Randomness=false, Non Uniform Scale=<1 1 -1>, Sphere Radius=28, Surface Only=false)
   &AddVelocity(Velocity=&UniformRangedVector(Minimum=<-50 -50 100>, Maximum=<50 50 200>))
  ]
  particleUpdateScript [
   &ParticleState()
   &GravityForce()
   &SolveForcesAndVelocity(Clamp Velocity=false, Limit Acceleration=false)
   &ScaleSpriteSize(Scale Factor=&Vector2DFromCurve(
     Vector2Curve=DataInterface() {propertyValues {
       XCurve {keys [
         {arriveTangent=5.9728, interpMode=RCIM_Cubic, leaveTangent=5.9728, tangentMode=RCTM_User, time=0, value=0.5817}
         {arriveTangent=-0.1, interpMode=RCIM_Cubic, leaveTangent=-0.1, tangentMode=RCTM_User, time=0.173, value=0.9552}
         {interpMode=RCIM_Cubic, tangentMode=RCTM_User, time=1, value=0.1086}
        ]}
       YCurve {keys [
         {arriveTangent=6.2715, interpMode=RCIM_Cubic, leaveTangent=6.2715, tangentMode=RCTM_User, time=0, value=0.5817}
         {arriveTangent=1.8836, interpMode=RCIM_Cubic, leaveTangent=1.8836, tangentMode=RCTM_User, time=0.209, value=1.5556}
         {arriveTangent=0.0517, interpMode=RCIM_Cubic, leaveTangent=0.0517, tangentMode=RCTM_User, time=1, value=2.8163}
        ]}
      }}
     ScaleCurve=<1 1>
    ))
  ]
  rendererClasses [&NiagaraSpriteRendererProperties]
  renderers [{rendererClass=&NiagaraSpriteRendererProperties}]
 }
 Light_Burst {
  emitterUpdateScript [
   &EmitterState(
    Inactive Response=::NewEnumerator1
    Life Cycle Mode=::NewEnumerator1
    Loop Behavior=::NewEnumerator1
    Loop Delay=0.5
    Loop Duration Mode=::NewEnumerator1
    UseLoopDelay=true
   )
   &SpawnBurst_Instantaneous(Spawn Count=1, Use Spawn Probability=false)
  ]
  particleSpawnScript [
   &InitializeParticle(
    Color=<rgba 8 4.8 1.6 1>
    Lifetime=&UniformRangedFloat(Minimum=0.1, Maximum=0.15)
    Mass=&UniformRangedFloat(Minimum=0.4, Maximum=2)
    Mesh Scale=<2 2 2>
    Module.Write Color=true
    Module.Write Lifetime=true
    Module.Write Mass=false
    Module.Write Material Random=true
    Module.Write Position=false
    Module.Write Scale=true
    Module.Write SpriteRotation=true
    Module.Write SpriteSize=false
    Position=&SimulationPosition()
    Sprite Rotation=&NormalizedAngleToDegrees(NormalizedAngle=&UniformRangedFloat(Minimum=0, Maximum=1))
    Sprite Size=<100 100>
   )
   &ApplyInitialForces()
   &FN_Light_v2(Light Radius=100)
   &SphereLocation(MaskSpawnGroup=false, Module.Override Randomness=false, Offset=<0 0 -20>, Sphere Radius=0, Surface Only=false)
  ]
  particleUpdateScript [
   &ParticleState()
   &SolveForcesAndVelocity(Clamp Velocity=false, Limit Acceleration=false)
   &SolveRotationalForcesAndVelocity()
  ]
  rendererClasses [&NiagaraLightRendererProperties]
  renderers [{rendererClass=&NiagaraLightRendererProperties, rendererIndex=0}]
 }
 Lightning1 {
  emitterUpdateScript [
   &EmitterState(
    Inactive Response=::NewEnumerator1
    Life Cycle Mode=::NewEnumerator1
    Loop Delay=$System.New float
    Loop Duration=0.26
    Recalculate Duration Each Loop=true
    UseLoopDelay=true
   )
   &SpawnBurst_Instantaneous(Spawn Count=1, Use Spawn Probability=false)
  ]
  particleSpawnScript [
   &InitializeParticle(
    Color=<rgba 6 4.0433 0 1>
    Lifetime=&UniformRangedFloat(Minimum=0.1, Maximum=0.15)
    Mass=&UniformRangedFloat(Minimum=0.4, Maximum=2)
    Mesh Scale=<0.6 0.6 0.6>
    Module.Write Color=true
    Module.Write Lifetime=true
    Module.Write Mass=false
    Module.Write Material Random=true
    Module.Write Position=false
    Module.Write Scale=true
    Module.Write SpriteRotation=true
    Module.Write SpriteSize=false
    Position=&SimulationPosition()
    Sprite Rotation=&NormalizedAngleToDegrees(NormalizedAngle=&UniformRangedFloat(Minimum=0, Maximum=1))
    Sprite Size=<100 100>
   )
   &ApplyInitialForces()
   &CylinderLocation(
    Cylinder Height=0
    Cylinder Radius=10
    MaskSpawnGroup=false
    Module.Enable Lathe Profile=false
    Module.Override Randomness=false
    Module.Surface Only=true
    Offset=<0 0 -6>
    Surface Only Band Thickness=1
   )
   /Niagara/Modules/Spawn/Orientation/V2/InitialMeshOrientation(
    Orientation Vector=&UniformRangedVector(Minimum=<-0.25 -0.25 0>, Maximum=<0.25 0.25 0>)
    Rotation=&UniformRangedVector(Minimum=<0 0 0>, Maximum=<0 0 0>)
    Use Orientation Vector=true
    Use Rotation Vector=false
   ) {module=InitialMeshOrientation001}
   /Niagara/Modules/Audio/PlayAudio(
    Audio Player=DataInterface() {propertyValues {
      Attenuation=null
      Concurrency=null
      MaxPlaysPerTick=10
      SoundToPlay=/Game/Athena/Sounds/Emotes/BadMood/Emote_BadMood_Lightning_OneShots
      bAllowLoopingOneShotSounds=false
      bLimitPlaysPerTick=true
      bOnlyActiveDuringGameplay=false
      bStopWhenComponentIsDestroyed=true
      ConfigurationUserParameter {parameter {name=None, typeDefHandle {registeredTypeIndex=1}, varData []}}
      ParameterNames []
     }}
    PlayAudio=true
   )
  ]
  particleUpdateScript [
   &ParticleState()
   &ScaleMeshSize(Scale Factor=&VectorFromFloat(Value=&MakeFloatFromVector(
      Vector=&VectorFromCurve(VectorCurve=DataInterface() {propertyValues {
         XCurve {keys [
           <at 0 0 RCIM_Cubic>
           {arriveTangent=1.0549, interpMode=RCIM_Cubic, leaveTangent=1.0549, time=0.3006, value=0.6597}
           {interpMode=RCIM_Cubic, leaveTangent=-0.3378, time=0.8107, value=0.8553}
          ]}
         YCurve {keys [<at -0.0011 1.0067>]}
         ZCurve {keys [<at -0.0011 1.0067>]}
        }})
      Channel=::NewEnumerator0
     )))
   &SolveForcesAndVelocity(Clamp Velocity=false, Limit Acceleration=false)
   &SolveRotationalForcesAndVelocity()
  ]
  rendererClasses [&NiagaraMeshRendererProperties]
  renderers [{rendererClass=&NiagaraMeshRendererProperties, rendererIndex=0}]
 }
 Lightning_Burst {
  emitterUpdateScript [
   &EmitterState(
    Inactive Response=::NewEnumerator1
    Life Cycle Mode=::NewEnumerator1
    Loop Behavior=::NewEnumerator1
    Loop Delay=0.9
    Loop Duration=0.26
    UseLoopDelay=true
   )
   &SpawnBurst_Instantaneous(Spawn Count=1, Use Spawn Probability=false)
  ]
  particleSpawnScript [
   &InitializeParticle(
    Color=<rgba 6 4.0433 0 1>
    Lifetime=&UniformRangedFloat(Minimum=0.1, Maximum=0.15)
    Mass=&UniformRangedFloat(Minimum=0.4, Maximum=2)
    Mesh Scale=<0.6 0.6 0.6>
    Module.Write Color=true
    Module.Write Lifetime=true
    Module.Write Mass=false
    Module.Write Material Random=true
    Module.Write Position=false
    Module.Write Scale=true
    Module.Write SpriteRotation=true
    Module.Write SpriteSize=false
    Position=&SimulationPosition()
    Sprite Rotation=&NormalizedAngleToDegrees(NormalizedAngle=&UniformRangedFloat(Minimum=0, Maximum=1))
    Sprite Size=<100 100>
   )
   &ApplyInitialForces()
   &CylinderLocation(
    Cylinder Height=0
    Cylinder Radius=10
    MaskSpawnGroup=false
    Module.Enable Lathe Profile=false
    Module.Override Randomness=false
    Module.Surface Only=true
    Offset=<0 0 -6>
    Surface Only Band Thickness=1
   )
   &InitialMeshOrientation(
    Orientation Vector=&UniformRangedVector(Minimum=<-0.25 -0.25 0>, Maximum=<0.25 0.25 0>)
    Rotation=&UniformRangedVector(Minimum=<0 0 0>, Maximum=<0 0 0>)
    Use Orientation Vector=true
    Use Rotation Vector=false
   ) {module=InitialMeshOrientation001}
  ]
  particleUpdateScript [
   &ParticleState()
   &ScaleMeshSize(Scale Factor=&VectorFromFloat(Value=&MakeFloatFromVector(
      Vector=&VectorFromCurve(VectorCurve=DataInterface() {propertyValues {
         XCurve {keys [
           <at 0 0 RCIM_Cubic>
           {arriveTangent=1.0549, interpMode=RCIM_Cubic, leaveTangent=1.0549, time=0.3006, value=0.6597}
           {interpMode=RCIM_Cubic, leaveTangent=-0.3378, time=0.8107, value=0.8553}
          ]}
         YCurve {keys [<at -0.0011 1.0067>]}
         ZCurve {keys [<at -0.0011 1.0067>]}
        }})
      Channel=::NewEnumerator0
     )))
   &SolveForcesAndVelocity(Clamp Velocity=false, Limit Acceleration=false)
   &SolveRotationalForcesAndVelocity()
  ]
  rendererClasses [&NiagaraMeshRendererProperties]
  renderers [{rendererClass=&NiagaraMeshRendererProperties, rendererIndex=0}]
 }
 Lights {
  emitterUpdateScript [
   &EmitterState(
    Inactive Response=::NewEnumerator1
    Life Cycle Mode=::NewEnumerator1
    Loop Delay=$System.New float
    Loop Duration=0.26
    Recalculate Duration Each Loop=true
    UseLoopDelay=true
   )
   &SpawnBurst_Instantaneous(Spawn Count=1, Use Spawn Probability=false)
  ]
  particleSpawnScript [
   &InitializeParticle(
    Color=<rgba 8 4.8 1.6 1>
    Lifetime=&UniformRangedFloat(Minimum=0.1, Maximum=0.15)
    Mass=&UniformRangedFloat(Minimum=0.4, Maximum=2)
    Mesh Scale=<2 2 2>
    Module.Write Color=true
    Module.Write Lifetime=true
    Module.Write Mass=false
    Module.Write Material Random=true
    Module.Write Position=false
    Module.Write Scale=true
    Module.Write SpriteRotation=true
    Module.Write SpriteSize=false
    Position=&SimulationPosition()
    Sprite Rotation=&NormalizedAngleToDegrees(NormalizedAngle=&UniformRangedFloat(Minimum=0, Maximum=1))
    Sprite Size=<100 100>
   )
   &ApplyInitialForces()
   &FN_Light_v2(Light Radius=125)
   &SphereLocation(MaskSpawnGroup=false, Module.Override Randomness=false, Offset=<0 0 -20>, Sphere Radius=0, Surface Only=false)
  ]
  particleUpdateScript [
   &ParticleState()
   &SolveForcesAndVelocity(Clamp Velocity=false, Limit Acceleration=false)
   &SolveRotationalForcesAndVelocity()
  ]
  rendererClasses [&NiagaraLightRendererProperties]
  renderers [{rendererClass=&NiagaraLightRendererProperties, rendererIndex=0}]
 }
 Rain2 {
  emitterUpdateScript [
   &EmitterState(
    Delay First Loop Only=true
    Inactive Response=::NewEnumerator1
    Life Cycle Mode=::NewEnumerator1
    Loop Delay=0.5
    Loop Duration=1
    UseLoopDelay=true
   )
   &SpawnRate(SpawnRate=30, Use Spawn Probability=false)
  ]
  particleSpawnScript [
   &InitializeParticle(
    Color=<rgba 1 1 1 0.35>
    Lifetime=&UniformRangedFloat(Minimum=0.6, Maximum=0.8)
    Mass=&UniformRangedFloat(Minimum=0.75, Maximum=1.75)
    Module.Write Color=true
    Module.Write Lifetime=true
    Module.Write Mass=false
    Module.Write Material Random=true
    Module.Write Position=false
    Module.Write Scale=false
    Module.Write SpriteRotation=false
    Module.Write SpriteSize=true
    Position=&SimulationPosition()
    Sprite Size=&Vector2DFromFloat(Value=&MakeFloatFromVector(Vector=&UniformRangedVector(Minimum=<6 0 0>, Maximum=<7 1 1>), Channel=::NewEnumerator0))
   )
   &SphereLocation(Hemisphere Z=true, MaskSpawnGroup=false, Module.Override Randomness=false, Offset=<0 0 -10>, Sphere Radius=28, Surface Only=false)
  ]
  particleUpdateScript [
   &ParticleState()
   &GravityForce()
   &Collision(
    Advanced Aging Rate=1
    Age Colliding Particles=false
    Collision Enabled=false
    Module.Control Rest State Via Surface Normal=false
    Module.EnableMaxCollisionCount=false
    Module.Randomize Collision Normal Vector=false
    Radius Calculation Type=::NewEnumerator1
    Restitution=0
   )
   &SolveForcesAndVelocity(Clamp Velocity=false, Limit Acceleration=false)
   &ScaleSpriteSize(Scale Factor=&Vector2DFromCurve(
     Vector2Curve=DataInterface() {propertyValues {
       XCurve {keys [
         {arriveTangent=5.9728, interpMode=RCIM_Cubic, leaveTangent=5.9728, tangentMode=RCTM_User, time=0, value=0.5817}
         {arriveTangent=-0.1, interpMode=RCIM_Cubic, leaveTangent=-0.1, tangentMode=RCTM_User, time=0.173, value=0.9552}
         {interpMode=RCIM_Cubic, tangentMode=RCTM_User, time=1, value=0.2}
        ]}
       YCurve {keys [
         {arriveTangent=6.2715, interpMode=RCIM_Cubic, leaveTangent=6.2715, tangentMode=RCTM_User, time=0, value=0.5817}
         {arriveTangent=1.8836, interpMode=RCIM_Cubic, leaveTangent=1.8836, tangentMode=RCTM_User, time=0.209, value=1.5556}
         {arriveTangent=0.0517, interpMode=RCIM_Cubic, leaveTangent=0.0517, tangentMode=RCTM_User, time=1, value=2.8163}
        ]}
      }}
     ScaleCurve=<1 1>
    ))
  ]
  rendererClasses [&NiagaraSpriteRendererProperties]
  renderers [{rendererClass=&NiagaraSpriteRendererProperties}]
 }
}
summary {
 systemName=NS_BadMood_Emote_Storm_Frontend
 emitters [
  {emitterName=Collision_Splashes, rendererClasses [&NiagaraSpriteRendererProperties]}
  {emitterName=Lightning1, rendererClasses [&NiagaraMeshRendererProperties]}
  {emitterName=Lightning_Burst, rendererClasses [&NiagaraMeshRendererProperties]}
  {emitterName=Rain2, rendererClasses [&NiagaraSpriteRendererProperties]}
  {emitterName=Collision_Proxy, rendererClasses [&NiagaraSpriteRendererProperties]}
  {emitterName=Lights, rendererClasses [&NiagaraLightRendererProperties]}
  {emitterName=Light_Burst, rendererClasses [&NiagaraLightRendererProperties]}
  {emitterName=CloudLoop_Game, rendererClasses [&NiagaraMeshRendererProperties]}
 ]
 userVariables [
  User.bIsFrontEndPreview {type=NiagaraBool}
  User.SlurpColorAlpha {type=NiagaraFloat, defaultValue=0, description=""}
 ]
}
